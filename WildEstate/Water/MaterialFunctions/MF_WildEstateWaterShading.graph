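# Material function that shades Wild Estate water, outputting material attributes plus a mobile-only absorption value.
# Static switches pick or blend ocean, lake and river colors, tinted and desaturated by time of day.
# On the default path, scene-versus-pixel depth drives banded darkening, opacity and white intersection lines, with distance fades on color effects.
# Mobile uses sphere-mask opacity and absorption instead, two-sided sign separates underside values, and sun-glint nodes are unconnected.
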MaterialFunction /WildEstate/Water/MaterialFunctions/MF_WildEstateWaterShading
graph {nodes [
  Constant2Vector()
  Constant()
  Constant() {r=1}
  FunctionInput(Preview=StaticBool()) {bUsePreviewValueAsDefault=true, inputName=IsHLOD Material, inputType=FunctionInput_StaticBool, sortPriority=16561}
  n1 = Saturate(None=TwoSidedSign())
  n2 = NamedRerouteUsage()
  n3 = VectorParameter() {group=Shading, parameterName=WildEstate_Color_Ocean, defaultValue=<rgba 0.0058 0.1497 0.224 1>}
  n4 = VectorParameter() {group=Shading, parameterName=WildEstate_ColorRiver, defaultValue=<rgba 0.0058 0.1497 0.224 1>}
  n5 = FunctionInput(Preview=Constant()) {bUsePreviewValueAsDefault=true, inputName=TransitionMask, inputType=FunctionInput_Scalar, sortPriority=10}
  n6 = VectorParameter() {group=Shading, parameterName=WildEstate_Color_Lake, defaultValue=<rgba 0.0058 0.1497 0.224 1>}
  n7 = StaticSwitchParameter(True=@n6.RGB, False=@n3.RGB) {parameterName=IsLake}
  n8 = FunctionInput(Preview=StaticBool()) {bUsePreviewValueAsDefault=true, inputName=IsRiver, inputType=FunctionInput_StaticBool, sortPriority=10}
  n9 = StaticSwitchParameter(
   True=LinearInterpolate(A=@n3.RGB, B=@n4.RGB, Alpha=@n5)
   False=StaticSwitchParameter(True=LinearInterpolate(A=@n6.RGB, B=@n4.RGB, Alpha=@n5), False=StaticSwitch(True=@n4.RGB, False=@n7, Value=@n8)) {parameterName=River Transition}
  ) {parameterName=Ocean River Transition}
  n10 = Subtract(A=SceneDepthWithoutWater!(UVs).R, B=PixelDepth().R)
  n11 = Min(
   A=@n10
   B=FunctionInput(Preview=Constant() {r=99999997952}) {bUsePreviewValueAsDefault=true, inputName=Depth, inputType=FunctionInput_Scalar, sortPriority=50}
  )
  n12 = Divide(A=@n11, B=ScalarParameter() {defaultValue=64, parameterName=DepthColorEdgeBandingWith})
  n13 = ScalarParameter() {defaultValue=3, parameterName=DepthColorEdgeSteps}
  n14 = Divide(A=Round(None=Multiply(A=@n12, B=@n13)), B=@n13)
  n15 = PixelDepth()
  n16 = Saturate(None=Divide(A=Subtract(A=@n15.R) {constB=10000}) {constB=10000})
  n17 = LinearInterpolate(A=LinearInterpolate(A=ScalarParameter() {defaultValue=1.3, parameterName=DeepwaterDarken}, Alpha=Saturate(None=@n14)) {constA=1}, Alpha=@n16)
  n18 = OneMinus(None=@n16)
  n19 = ScalarParameter() {defaultValue=1.0357, parameterName=IntersectionLine_DepthOffset}
  n20 = FunctionInput() {inputName=Normal}
  n21 = DotProduct(A=@n20, B=CameraVectorWS())
  n22 = Round(None=Subtract(A=@n10, B=Divide(A=@n19, B=@n21)))
  n23 = Multiply(A=ScalarParameter() {defaultValue=0.4, parameterName=IntersectionLine_Opacity}, B=OneMinus(None=Saturate(None=@n22)))
  n24 = Desaturation(
   None=LinearInterpolate(
    A=Multiply(A=@n9, B=@n17)
    B=VectorParameter() {parameterName=IntersectionWhitelines, defaultValue=<rgba 1 1 1 1>}.RGB
    Alpha=Multiply(A=@n18, B=@n23)
   )
   Fraction=NamedRerouteUsage()
  )
  n25 = SphereMask(
   A=WorldPosition().XY
   B=ComponentMask(None=VectorParameter() {parameterName=SpheremaskCenter, defaultValue=<rgba 0 0 0 1>}.RGB) {g=true, r=true}
   Radius=ScalarParameter() {defaultValue=185112.3125, parameterName=SphereMaskRadius}
   Hardness=ScalarParameter() {defaultValue=0.6279, parameterName=SphereMaskHardness}
  )
  n26 = LinearInterpolate(B=ScalarParameter() {defaultValue=0.4, parameterName=OpacityMobile}, Alpha=@n25) {constA=1}
  n27 = Multiply(A=NamedRerouteUsage(), B=Desaturation(None=@n9, Fraction=NamedRerouteUsage()))
  n28 = ScalarParameter() {defaultValue=1000, parameterName=DepthOpacityFade}
  n29 = LinearInterpolate(A=ScalarParameter() {defaultValue=0.9, group=Shading, parameterName=Opacity_WildEstate}, Alpha=Saturate(None=Divide(A=@n10, B=@n28)))
  n30 = SetMaterialAttributes!(
   MaterialAttributes
   Specular=LinearInterpolate(
    A=ScalarParameter() {group=Shading, parameterName=Specular_WildEstate_Bottom}
    B=ScalarParameter() {group=Shading, parameterName=Specular_WildEstate}
    Alpha=@n1
   )
   BaseColor=ShadingPathSwitch(Default=Multiply(A=@n2, B=@n24), Mobile=Divide(A=@n27, B=@n26))
   Opacity=LinearInterpolate(
    A=ScalarParameter() {defaultValue=0.0336, group=Shading, parameterName=Opacity_WildEstate_Bottom}
    B=ShadingPathSwitch(Default=@n29, Mobile=@n26)
    Alpha=@n1
   )
   EmissiveColor
   Roughness=Constant() {r=0.85}
   Metallic=Constant()
   WorldPositionOffset=FunctionInput() {inputName=WPO}
   Normal=@n20
  ) {attributeSetTypes [
    #1
    #2
    #3
    #4
    #5
    #6
    #7
    #8
   ]}
  FunctionOutput(None=@n30)
  n31 = VectorParameter() {parameterName=Absorption_Mobile, defaultValue=<rgba 0.9635 0.9738 1 0.0075>}
  n32 = LinearInterpolate(B=Multiply(A=@n31.RGB, B=@n31.RGB), Alpha=@n25)
  n33 = Saturate(None=TwoSidedSign())
  FunctionOutput(None=ShadingPathSwitch(Default=Constant(), Mobile=LinearInterpolate(B=@n32, Alpha=@n33))) {outputName=Absorption, sortPriority=99}
  Max()
  n34 = CollectionParameter() {
   collection=&FortniteMaterialParameters
   parameterName=SunlightDirectionalVector
  }
  n35 = DotProduct(A=ComponentMask(None=@n34) {b=true, g=true, r=true}, B=ReflectionVectorWS(CustomWorldNormal=PixelNormalWS()))
  n36 = ScalarParameter() {defaultValue=0.974, group=Shading, parameterName=SpecularSunMask}
  n37 = Saturate(None=Divide(A=Subtract(A=@n35, B=@n36), B=ScalarParameter() {defaultValue=0.01, group=Shading, parameterName=SpecularSunMaskBlend}))
  Multiply(A=@n37, B=VectorParameter() {parameterName=SunColor, defaultValue=<rgba 1 0.5298 0.2253 1>}.RGB)
  NamedRerouteDeclaration(None=MaterialFunctionCall(
    Color_Morning=VectorParameter() {parameterName=Morning, defaultValue=<rgba 0.4897 1.1 0.8641 1>}.RGB
    Color_Day=VectorParameter() {parameterName=Day, defaultValue=<rgba 1 1 1 1>}.RGB
    Color_Evening=VectorParameter() {parameterName=Evening, defaultValue=<rgba 0.321 1 0.9441 1>}.RGB
    Color_Night=VectorParameter() {parameterName=Night, defaultValue=<rgba 0.7786 1.1 0.6532 1>}.RGB
   ) {materialFunction=&Color_Time_of_Day}.Result) {name=TODM_Tint, nodeColor=<rgba 0.8984 1 0 1>}
  NamedRerouteDeclaration(None=MaterialFunctionCall(
    Value_Morning=ScalarParameter() {defaultValue=0.5, parameterName=Dest_Morning}
    Value_Day=ScalarParameter() {parameterName=Dest_Day}
    Value_Evening=ScalarParameter() {defaultValue=0.488, parameterName=Dest_Evening}
    Value_Night=ScalarParameter() {defaultValue=0.48, parameterName=Dest_Night}
   ) {materialFunction=&Scalar_Time_of_Day}.Result) {name=TODM_Desat, nodeColor=<rgba 0.8984 1 0 1>}
  OneMinus(None=@n37)
  Power(Base=@n35, Exp=@n36)
  n38 = ScalarParameter() {defaultValue=16, parameterName=DepthColorEdge}
  Saturate(None=Ceil(None=Subtract(A=@n11, B=@n38)))
  StaticSwitchParameter() {parameterName=IsFarDistance Water}
 ]}
