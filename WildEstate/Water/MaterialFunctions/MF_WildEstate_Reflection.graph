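# Water material function for WildEstate that adds a fake grazing-angle reflection to incoming material attributes.
# A sharp Fresnel term darkens base color and adds emissive tinted by the CloudsHeightFog collection parameter, both limited to outside a parameterized XY sphere mask.
# A square falloff around the world origin fades World Position Offset to zero at distance and is also output as "WPO mask".
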MaterialFunction /WildEstate/Water/MaterialFunctions/MF_WildEstate_Reflection
graph {nodes [
  n1 = GetMaterialAttributes(None=FunctionInput() {inputName=Material, inputType=FunctionInput_MaterialAttributes}) {attributeGetTypes [#1, #2, #3]}
  n2 = DotProduct(A=VertexNormalWS(), B=CameraVectorWS())
  n3 = Power(Base=OneMinus(None=Abs(None=@n2))) {constExponent=32}
  n4 = Multiply(A=Multiply(A=@n1.BaseColor) {constB=0.8}, B=OneMinus(None=@n3))
  n5 = VectorParameter() {parameterName=ReflectionMask Center, defaultValue=<rgba 0 0 0 1>}
  n6 = OneMinus(None=SphereMask(
    A=ComponentMask(None=@n5.RGB) {g=true, r=true}
    B=WorldPosition().XY
    Radius=ScalarParameter() {defaultValue=193457.2031, parameterName=ReflectionmaskRadius}
    Hardness=ScalarParameter() {defaultValue=0.1, parameterName=Reflectionmaskhardness}
   ))
  n7 = CollectionParameter() {collection=&FortniteMaterialParameters, parameterName=CloudsHeightFog}
  n8 = Multiply(A=Multiply(A=@n3, B=ComponentMask(None=@n7) {b=true, g=true, r=true}), B=@n6)
  n9 = WorldPosition()
  n10 = Multiply(A=Subtract(A=Abs(None=@n9.XY)) {constB=60000}) {constB=0.0001}
  n11 = ComponentMask(None=@n10) {r=true}
  n12 = ComponentMask(None=@n10) {g=true}
  n13 = OneMinus(None=Saturate(None=Max(A=@n11, B=@n12)))
  FunctionOutput(None=SetMaterialAttributes!(
    MaterialAttributes=@n1.MaterialAttributes
    BaseColor=LinearInterpolate(A=@n1.BaseColor, B=@n4, Alpha=@n6)
    EmissiveColor=Add(A=@n1.EmissiveColor, B=@n8)
    WorldPositionOffset=Multiply(A=@n1.WorldPositionOffset, B=@n13)
   ) {attributeSetTypes [#1, #3, #2]})
  FunctionOutput(None=@n13) {outputName=WPO mask, sortPriority=99}
 ]}
