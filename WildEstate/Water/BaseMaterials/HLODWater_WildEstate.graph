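# Single Layer Water material for WildEstate's HLOD water, standing in for the water surface in distant proxies.
# It calls the shared WildEstate water shading function with the HLOD switch on, a flat normal and zero world position offset, then a reflection function.
# Further branches blend opacity by pixel depth and compute a camera-height vector blended by distance against a Distant Cutoff parameter.
# It also includes a water quality function noted as simplifying specular and roughness on mobile.
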Material /WildEstate/Water/BaseMaterials/HLODWater_WildEstate
graph {
 properties {
  allowTranslucentCustomDepthWrites=true
  bComputeFogPerPixel=true
  bUsedWithInstancedStaticMeshes=true
  bUsedWithNanite=true
  bUsedWithSplineMeshes=true
  bUsedWithWater=true
  pixelDepthOffsetMode=PDOM_Legacy
  shadingModel=MSM_SingleLayerWater
  twoSided=true
  usedShadingModels=SingleLayerWater
 }
 nodes [
  n1 = MaterialFunctionCall!(
   Normal=Constant3Vector() {constant=<rgba 0 0 1 0>}
   WPO=Constant3Vector()
   TransitionMask
   IsRiver
   Depth
   IsHLOD Material=StaticBool() {value=true}
  ) {materialFunction=/WildEstate/Water/MaterialFunctions/MF_WildEstateWaterShading}
  GetMaterialAttributes(None=MaterialFunctionCall!(Material=@n1.Result) {materialFunction=/WildEstate/Water/MaterialFunctions/MF_WildEstate_Reflection}.Result) {attributeGetTypes [
    #1
    #2
    #3
    #4
    #5
    #6
   ]}
  n2 = CameraPositionWS()
  n3 = Distance(A=WorldPosition() {worldPositionShaderOffset=WPT_CameraRelative}.XYZ, B=Constant())
  n4 = ScalarParameter() {defaultValue=600000, group=Outlines Opacity, parameterName=Distant Cutoff}
  LinearInterpolate(B=AppendVector(A=Constant2Vector(), B=ComponentMask(None=@n2) {b=true}), Alpha=Saturate(None=Subtract(A=@n3, B=@n4) {constB=20000}))
  MaterialFunctionCall!(Attributes) {desc=Simple Spec and Roughness On Mobile, materialFunction=/FortWater/Materials/Functions/WaterQualityOptimizations}
  n5 = GetMaterialAttributes(None=@n1.Result) {attributeGetTypes [#5]}
  n6 = Divide(A=Subtract(A=PixelDepth().R) {constB=100000}) {constB=10000}
  SetMaterialAttributes!(MaterialAttributes=@n5.MaterialAttributes, Opacity=LinearInterpolate(A=@n5.Opacity, Alpha=Saturate(None=@n6))) {attributeSetTypes [#5]}
  SingleLayerWaterMaterialOutput!(ScatteringCoefficients=Constant(), AbsorptionCoefficients, PhaseG, ColorScaleBehindWater)
 ]
}
