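# Material function that samples a world-space water biome color map, grouped under Mobile Water parameters.
# World position XY is shifted and scaled by BiomeTexture offset and scale parameters, then a Distortion input is added to form the UVs.
# The overridable texture input is sampled with a shared clamp sampler.
# A static switch outputs either the full RGBA sample or neutral gray RGB with the map's alpha retained.
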MaterialFunction /FortWater/Materials/Functions/MF_WaterBiomeColorMap
graph {nodes [
  n1 = FunctionInput(Preview=WorldPosition().XYZ) {bUsePreviewValueAsDefault=true, inputName=Position}
  n2 = AppendVector(
   A=ScalarParameter() {defaultValue=0.48, group=Mobile Water, parameterName=BiomeTexture OffsetX}
   B=ScalarParameter() {defaultValue=0.5, group=Mobile Water, parameterName=BiomeTexture OffsetY}
  )
  n3 = Divide(
   A=Add(A=ComponentMask(None=@n1) {g=true, r=true}, B=Multiply(A=@n2) {constB=-1})
   B=ScalarParameter() {defaultValue=350000, group=Mobile Water, parameterName=BiomeTexture Scale}
  ) {constB=350000}
  n4 = TextureObjectParameter() {
   group=Mobile Water
   parameterName=BiomeTexture
   samplerType=SAMPLERTYPE_LinearColor
   texture=/FortWater/Materials/WaterSurface/Helios/Textures/T_Water_Hermes_BiomeColorMap
  }
  n5 = TextureSample(
   UVs=Add(
    A=Add(A=@n3, B=@n2)
    B=FunctionInput() {bUsePreviewValueAsDefault=true, inputName=Distortion, inputType=FunctionInput_Vector2, sortPriority=1}
   )
   Tex=FunctionInput(Preview=@n4) {bUsePreviewValueAsDefault=true, inputType=FunctionInput_Texture2D, sortPriority=2}
  ) {
   samplerSource=SSM_Clamp_WorldGroupSettings
   samplerType=SAMPLERTYPE_LinearColor
   texture=/FortWater/Materials/WaterSurface/Helios/Textures/T_Water_Helios_BiomeColorMap
  }
  n6 = Constant3Vector() {constant=<rgba 0.5 0.5 0.5 0>}
  FunctionOutput(None=StaticSwitchParameter(True=@n5.RGBA, False=AppendVector(A=@n6, B=@n5.A)) {defaultValue=true, group=Mobile Water, parameterName=Mobile Use Biome Color Map}) {outputName="Biome Color Map (RGBA)"}
 ]}
