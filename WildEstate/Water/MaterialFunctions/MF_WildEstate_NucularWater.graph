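# Material function that overrides water base color and emissive with stylized purple-blue "Nucular" water inside a scaled radial world-space zone, toggled by UseNucularWater.
# Two panning swirl textures distort depth-based stepped bands with outlines, with river depth ranges and simpler mobile color.
# The swirls also break up the zone edge through HeightLerp, while time-of-day colors tint the emissive.
# Extra outputs expose per-pixel and camera-position zone masks, the latter for post.
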MaterialFunction /WildEstate/Water/MaterialFunctions/MF_WildEstate_NucularWater
graph {nodes [
  n1 = ComponentMask(None=FunctionInput(Preview=Constant3Vector()) {bUsePreviewValueAsDefault=true, inputName=FlowMap Input, sortPriority=30}) {r=true}
  n2 = Divide(A=WorldPosition().XY, B=ScalarParameter() {defaultValue=3865.9348, parameterName=Nucular_SwirlSize})
  n3 = Constant2Vector() {g=-1.15, r=1.1}
  n4 = ScalarParameter() {defaultValue=-0.03, parameterName=Nucular_Swirl_PanSpeed}
  n5 = TextureSample(UVs=Panner(Coordinate=Multiply(A=@n2, B=@n3) {constB=1.1}, Speed=Multiply(A=@n4) {constB=-1}) {bFractionalPart=true}) {samplerType=SAMPLERTYPE_Grayscale, texture=/WildEstate/Water/Textures/WildEstate_WaterSwirls_M}
  n6 = TextureSample(UVs=Panner(Coordinate=@n2, Speed=@n4) {bFractionalPart=true}) {samplerType=SAMPLERTYPE_Grayscale, texture=&WildEstate_WaterSwirls_M}
  n7 = MaterialFunctionCall(Base=@n5.R, Blend=@n6.R) {materialFunction=&Blend_Overlay}
  Add(A=Multiply(A=@n1, B=ScalarParameter() {defaultValue=0.208, parameterName=Nucular_FlowMap_Distort}), B=@n7.Result)
  n8 = WorldPosition()
  n9 = ComponentMask(None=VectorParameter() {parameterName=Nucular_SphereMask_Center, defaultValue=<rgba 11987.2383 486.8101 0 1>}.RGB) {g=true, r=true}
  n10 = Subtract(A=@n8.XY, B=@n9)
  n11 = ComponentMask(None=VectorParameter() {parameterName=Nucular_SphereMask_Scale, defaultValue=<rgba 0.7 1 1 1>}.RGB) {g=true, r=true}
  n12 = ScalarParameter() {defaultValue=40168.957, parameterName=Nucular_SphereMask_Radius}
  n13 = Subtract(A=Length(None=Multiply(A=@n10, B=@n11)) {is=class MaterialExpressionLength}, B=@n12)
  n14 = ScalarParameter() {defaultValue=5000, parameterName=Nucular_SphereMask_BlendDistance}
  n15 = OneMinus(None=Saturate(None=Divide(A=@n13, B=@n14)))
  n16 = MaterialFunctionCall(
   Transition Phase=@n15
   Height Texture=OneMinus(None=@n7.Result)
   Contrast=ScalarParameter() {defaultValue=0.2111, parameterName=Nucular_SphereMask_Contrast}
  ) {materialFunction=&HeightLerp}
  Ceil(None=@n16.Alpha)
  Constant()
  Distance(A=@n9, B=@n8.XY)
  n17 = ComponentMask(None=CameraPositionWS()) {g=true, r=true}
  Distance(A=@n9, B=@n17)
  n18 = LinearInterpolate(
   A=VectorParameter() {group=Nucular, parameterName=Nucular_Color1, defaultValue=<rgba 0.5729 0 0.7396 1>}.RGB
   B=VectorParameter() {group=Nucular, parameterName=Nucular_Color2, defaultValue=<rgba 0.1253 0.0146 0.2344 1>}.RGB
   Alpha=NamedRerouteUsage()
  )
  n19 = ScalarParameter() {defaultValue=0.344, parameterName=Nucular_BlueBlend}
  n20 = FunctionInput() {inputName=Depth, inputType=FunctionInput_Scalar, sortPriority=20}
  n21 = ScalarParameter() {defaultValue=881.1371, group=Nucular, parameterName=Nucular_DepthRange}
  n22 = ScalarParameter() {defaultValue=415.1995, group=Nucular, parameterName=Nucular_DepthRange_River}
  n23 = LinearInterpolate(
   A=@n21
   B=@n22
   Alpha=FunctionInput(Preview=Constant()) {bUsePreviewValueAsDefault=true, inputName=Transition_Mask, inputType=FunctionInput_Scalar, sortPriority=20}
  )
  n24 = StaticBool()
  n25 = FunctionInput(Preview=@n24) {bUsePreviewValueAsDefault=true, inputName=isRiver_Transition, inputType=FunctionInput_StaticBool, sortPriority=20}
  n26 = Divide(
   A=@n20
   B=StaticSwitch(
    True=StaticSwitch(True=@n23, False=@n22, Value=@n25)
    False=@n21
    Value=FunctionInput(Preview=@n24) {bUsePreviewValueAsDefault=true, inputName=isRiver, inputType=FunctionInput_StaticBool, sortPriority=20}
   )
  )
  n27 = Multiply(A=Multiply(A=OneMinus(None=@n26), B=@n26)) {constB=4}
  n28 = ScalarParameter() {defaultValue=1.5, group=Nucular, parameterName=Nucular_DepthFalloff}
  n29 = Saturate(None=Divide(A=@n20, B=ScalarParameter() {defaultValue=50, parameterName=FadeDistortion}))
  n30 = Multiply(A=Subtract(A=@n7.Result) {constB=0.5}, B=@n29)
  n31 = ScalarParameter() {defaultValue=-0.5779, group=Nucular, parameterName=Nucular_SwirlAdd}
  n32 = ScalarParameter() {defaultValue=4, group=Nucular, parameterName=Nucular_Steps}
  n33 = Multiply(A=Add(A=Power(Base=@n27, Exp=@n28), B=Multiply(A=@n30, B=@n31)), B=@n32)
  n34 = Frac(None=@n33)
  n35 = OneMinus(None=@n34)
  n36 = ScalarParameter() {defaultValue=0.05, group=Nucular, parameterName=Nucular_SoftFrac}
  n37 = Multiply(A=@n34, B=Saturate(None=Divide(A=@n35, B=@n36) {constB=0.1}))
  n38 = ShadingPathSwitch(
   Default=ScalarParameter() {defaultValue=0.05, group=Nucular, parameterName=Nucular_Steps_RampWidth}
   Mobile=ScalarParameter() {defaultValue=0.05, group=Nucular, parameterName=Nucular_Steps_RampWidth_Mobile}
  )
  n39 = Divide(
   A=Subtract(A=MaterialFunctionCall!(In=@n33, Ramp Width=@n38) {materialFunction=/Engine/Functions/Engine_MaterialFunctions01/Math/SmoothCeil}.Result)
   B=@n32
  )
  n40 = Saturate(None=Add(A=@n39, B=ScalarParameter() {defaultValue=-0.44, parameterName=Nucular_BlueBias}))
  n41 = Multiply(A=@n37, B=Multiply(A=OneMinus(None=@n40), B=@n39))
  n42 = LinearInterpolate(
   A=@n18
   B=VectorParameter() {group=Nucular, parameterName=Nucular_Color_Blue, defaultValue=<rgba 0.052 0.0037 0.7865 1>}.RGB
   Alpha=Saturate(None=Multiply(A=@n19, B=@n41))
  )
  Divide(A=@n42) {constB=0.4, desc=Mobile water opacity}
  Floor(None=@n33)
  Floor(None=@n16.Alpha)
  n43 = NamedRerouteUsage()
  n44 = ScalarParameter() {defaultValue=-0.3824, group=Nucular, parameterName=Nucular_Bias2}
  n45 = Multiply(A=Multiply(A=Subtract(A=@n43, B=@n44), B=OneMinus(None=@n43))) {constB=4}
  n46 = ScalarParameter() {defaultValue=-0.2331, group=Nucular, parameterName=Nucular_Bias}
  n47 = Multiply(A=Saturate(None=Subtract(A=@n45, B=@n46)), B=ScalarParameter() {defaultValue=3.9539, group=Nucular, parameterName=Nucular_Center})
  Floor(None=@n47)
  n48 = FunctionInput() {inputType=FunctionInput_MaterialAttributes}
  n49 = GetMaterialAttributes(None=@n48) {attributeGetTypes [#1, #2, #3]}
  n50 = LinearInterpolate(A=@n49.BaseColor, B=NamedRerouteUsage(), Alpha=@n16.Alpha)
  n51 = VectorParameter() {group=Nucular, parameterName=Nucular_Emissive_Morning, defaultValue=<rgba 0.785 0.3988 0.3872 1>}
  n52 = VectorParameter() {group=Nucular, parameterName=Nucular_Emissive_Day, defaultValue=<rgba 0.3438 0.3438 0.3438 1>}
  n53 = VectorParameter() {group=Nucular, parameterName=Nucular_Emissive_Evening, defaultValue=<rgba 0.785 0.3988 0.3872 1>}
  n54 = VectorParameter() {group=Nucular, parameterName=Nucular_Emissive_Night, defaultValue=<rgba 1 0.0579 0 1>}
  n55 = LinearInterpolate(
   A=@n49.EmissiveColor
   B=Multiply(
    A=NamedRerouteUsage()
    B=MaterialFunctionCall(Color_Morning=@n51.RGB, Color_Day=@n52.RGB, Color_Evening=@n53.RGB, Color_Night=@n54.RGB) {materialFunction=&Color_Time_of_Day}.Result
   )
   Alpha=@n16.Alpha
  )
  FunctionOutput(None=StaticSwitchParameter(
    True=SetMaterialAttributes!(MaterialAttributes=@n49.MaterialAttributes, BaseColor=@n50, EmissiveColor=@n55, Roughness=@n49.MaterialAttributes) {attributeSetTypes [#1, #2, #3]}
    False=@n48
   ) {parameterName=UseNucularWater})
  n56 = Multiply(A=Subtract(A=@n17, B=@n9), B=@n11)
  n57 = Divide(A=Subtract(A=Length(None=@n56), B=@n12), B=@n14)
  FunctionOutput(None=OneMinus(None=Saturate(None=@n57))) {outputName=MaskforPost, sortPriority=99}
  FunctionOutput(None=@n15) {outputName=MaskForPixelShader, sortPriority=9999}
  Max(A=@n5.R, B=@n6.R)
  Min(A=@n5.R, B=@n6.R)
  Multiply(A=@n7.Result, B=@n29)
  n58 = LinearInterpolate(A=ScalarParameter() {defaultValue=0.9518, parameterName=NucularGradientIntensity}, Alpha=OneMinus(None=@n37))
  NamedRerouteDeclaration(None=Saturate(None=Multiply(A=@n58, B=@n39))) {name=StepGradient, nodeColor=<rgba 1 0 0.6094 1>}
  n59 = Frac(None=@n47)
  n60 = Saturate(None=Divide(A=OneMinus(None=@n59), B=ScalarParameter() {defaultValue=0.05, group=Nucular, parameterName=Nucular_SoftFrac}) {constB=0.1})
  n61 = Frac(None=@n33)
  n62 = Multiply(A=@n61, B=OneMinus(None=@n61))
  n63 = Subtract(A=MaterialFunctionCall(In=@n47, Ramp Width=@n38) {materialFunction=&SmoothCeil}.Result)
  n64 = Subtract(
   A=LinearInterpolate(A=Multiply(A=@n59, B=@n60), B=Multiply(A=@n39, B=@n62), Alpha=Saturate(None=@n63))
   B=ScalarParameter() {defaultValue=0.0064, group=Nucular, parameterName=Nucular_Outlines}
  )
  n65 = LinearInterpolate(A=ScalarParameter() {defaultValue=0.6464, group=Nucular, parameterName=Nucular_Outline_Opacity}, Alpha=Saturate(None=Ceil(None=@n64)))
  n66 = ScalarParameter() {defaultValue=1.5, parameterName=Mobile BrightnessBoost}
  NamedRerouteDeclaration(None=ShadingPathSwitch(Default=Multiply(A=@n42, B=@n65), Mobile=Multiply(A=@n18, B=@n66))) {name=Color, nodeColor=<rgba 0 0.5078 1 1>}
  Round(None=@n16.Alpha)
  ScalarParameter() {defaultValue=0.8045, parameterName=Emissive}
  Subtract(A=@n1)
 ]}
