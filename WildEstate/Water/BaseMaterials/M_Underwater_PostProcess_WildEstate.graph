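# Underwater post-process material for WildEstate water that tints and fogs the scene, with waterline masks splitting the view.
# Desktop applies exponential depth extinction and fog, scaled by time of day and corrected with inverse eye adaptation.
# Mobile uses a cheaper color-map tint, depth gradient and scrolling two-layer caustics.
# Masks compare near-clip-plane positions against water body height plus Gerstner or mobile waves, while swamp or Nucular masks define biomes.
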Material /WildEstate/Water/BaseMaterials/M_Underwater_PostProcess_WildEstate
graph {
 properties {blendableLocation=BL_SceneColorAfterDOF, materialDomain=MD_PostProcess, pixelDepthOffsetMode=PDOM_Legacy, usedShadingModels=Default Lit}
 nodes [
  n1 = MaterialFunctionCall(
   Color_Morning=ScalarParameter() {defaultValue=0.75, parameterName=Morning}
   Color_Day=ScalarParameter() {defaultValue=1, parameterName=Day}
   Color_Evening=ScalarParameter() {defaultValue=0.75, parameterName=Evening}
   Color_Night=ScalarParameter() {defaultValue=0.45, parameterName=Night}
  ) {materialFunction=&Color_Time_of_Day}
  n2 = ScreenPosition()
  n3 = VectorParameter() {parameterName=EctinctionColor, defaultValue=<rgba 3 10 75 100>}
  n4 = VectorParameter() {parameterName=EctinctionColorMobile, defaultValue=<rgba 3.9062 10.8181 75 100>}
  n5 = ShadingPathSwitch(Default=Multiply(A=@n3.RGB, B=@n3.RGB), Mobile=Multiply(A=@n4.RGB, B=@n4.RGB))
  n6 = ScalarParameter() {defaultValue=1250, parameterName=Fog_Range}
  n7 = VectorParameter() {parameterName=EctinctionColor2, defaultValue=<rgba 3 2.5 2 50>}
  n8 = Multiply(A=@n7.RGB, B=@n7.RGB)
  n9 = ScalarParameter() {defaultValue=500, parameterName=Fog_Range2}
  n10 = DotProduct(
   A=CameraVectorWS()
   B=MaterialFunctionCall() {materialFunction=&CameraDirectionVector}.Result
  )
  n11 = Divide(A=ComponentMask(None=SceneTexture() {sceneTextureId=PPI_SceneDepth}.Color) {r=true}, B=Max(A=Abs(None=@n10)) {constB=0})
  n12 = ScalarParameter() {defaultValue=3, parameterName=Depth Lip Boost}
  n13 = Max(A=NamedRerouteUsage()) {constB=0}
  MaterialExpressionDivide_9 = Divide(
   A=LinearInterpolate(A=Multiply(A=@n11) {constB=0.5}, B=Multiply(A=@n12, B=@n13), Alpha=NamedRerouteUsage())
   B=LinearInterpolate(A=AppendVector(A=@n5, B=@n6), B=AppendVector(A=@n8, B=@n9), Alpha=NamedRerouteUsage())
  )
  n14 = Custom!(x=@MaterialExpressionDivide_9) {
   code="exp(-x);"
   desc="RGB = Extinction, A=Fog"
   description=exp
   outputType=CMOT_Float4
   inputs [{inputName=x, input {expression=&self:MaterialExpressionDivide_9}}]
  }
  n15 = ComponentMask(None=@n14) {b=true, g=true, r=true}
  n16 = Multiply(A=ComponentMask(None=SceneTexture(UVs=@n2.ViewportUV) {sceneTextureId=PPI_PostProcessInput0}.Color) {b=true, g=true, r=true}, B=@n15)
  n17 = Saturate(None=Multiply(A=ComponentMask(None=@n14) {a=true}, B=ScalarParameter() {defaultValue=1, parameterName=FogBlend}))
  n18 = VectorParameter() {parameterName=WaterFogColor_1, defaultValue=<rgba 0.212 0.6823 1 0.1>}
  n19 = VectorParameter() {parameterName=WaterFogColor_1_Mobile, defaultValue=<rgba 0.0833 0.6304 1 0.1>}
  n20 = VectorParameter() {parameterName=WaterFogColor_2, defaultValue=<rgba 20 17.5993 9.8343 0.03>}
  n21 = LinearInterpolate(
   A=ShadingPathSwitch(Default=Multiply(A=@n18.RGB, B=@n18.RGB), Mobile=Multiply(A=@n19.RGB, B=@n19.RGB))
   B=Multiply(A=@n20.RGB, B=@n20.RGB)
   Alpha=NamedRerouteUsage()
  )
  n22 = EyeAdaptationInverse(
   LightValueInput=Multiply(A=ComponentMask(None=@n1.Result) {r=true}, B=LinearInterpolate(A=@n21, B=@n16, Alpha=@n17))
   AlphaInput=ScalarParameter() {parameterName=ExposureEqualization}
  )
  n23 = ComponentMask(None=SceneTexture(UVs=ScreenPosition().ViewportUV) {sceneTextureId=PPI_PostProcessInput0}.Color) {b=true, g=true, r=true}
  n24 = NamedRerouteUsage()
  n25 = MaterialFunctionCall(
   Color_Morning=VectorParameter() {group=Mobile Water, parameterName=Mobile Color Morning, defaultValue=<rgba 1 0.6076 0.4972 1>}.RGB
   Color_Day=VectorParameter() {group=Mobile Water, parameterName=Mobile Color Day, defaultValue=<rgba 1 0.9153 0.832 1>}.RGB
   Color_Evening=VectorParameter() {group=Mobile Water, parameterName=Mobile Color Evening, defaultValue=<rgba 0.9635 0.5783 0.4424 1>}.RGB
   Color_Night=VectorParameter() {group=Mobile Water, parameterName=Mobile Color Night, defaultValue=<rgba 0.7292 0.7292 0.7292 1>}.RGB
  ) {materialFunction=&Color_Time_of_Day}
  n26 = SkyAtmosphereLightDirection()
  n27 = LinearInterpolate(
   A=@n25.Result
   B=LinearInterpolate(
    A=VectorParameter() {group=Mobile Water, parameterName=Mobile Color, defaultValue=<rgba 0.118 0.5899 0.7552 1>}.RGB
    B=VectorParameter() {group=Mobile Water, parameterName=Mobile Color_Biome, defaultValue=<rgba 0.118 0.5899 0.7552 1>}.RGB
    Alpha=NamedRerouteUsage()
   )
   Alpha=LinearInterpolate(Alpha=ComponentMask(None=@n26) {b=true}) {constA=0.85}
  )
  n28 = ComponentMask(None=SceneTexture() {sceneTextureId=PPI_PostProcessInput0}.Color) {b=true, g=true, r=true}
  n29 = CameraPositionWS()
  n30 = ComponentMask(None=NamedRerouteDeclaration(None=MaterialFunctionCall!(Position=@n29, Distortion, In) {materialFunction=/FortWater/Materials/Functions/MF_WaterBiomeColorMap}."Biome Color Map (RGBA)") {name=Color Map, nodeColor=<rgba 0.6875 1 0 1>}) {b=true, g=true, r=true}
  n31 = Add(A=Multiply(A=Multiply(A=@n27, B=@n28), B=NamedRerouteUsage()), B=Saturate(None=Add(A=@n30) {constB=-0.5}))
  n32 = Min(A=Multiply(A=@n30) {constB=2})
  n33 = ComponentMask(None=CameraVectorWS()) {b=true}
  n34 = Multiply(A=ComponentMask(None=SceneTexture() {sceneTextureId=PPI_SceneDepth}.Color) {r=true}) {constB=0.0002}
  n35 = Multiply(A=Saturate(None=Add(A=@n33) {constB=0.6}), B=SquareRoot(None=Saturate(None=@n34)))
  n36 = ScalarParameter() {defaultValue=1, group=Mobile Water, parameterName=MobileDepth}
  n37 = LinearInterpolate(
   A=Multiply(A=Multiply(A=@n31, B=@n32), B=@n25.Result)
   B=LinearInterpolate(
    A=VectorParameter() {group=Mobile Water, parameterName=Mobile Color Depth, defaultValue=<rgba 0.0113 0.0784 0.151 1>}.RGB
    B=VectorParameter() {group=Mobile Water, parameterName=Mobile Color Depth Biome, defaultValue=<rgba 0.0113 0.0784 0.151 1>}.RGB
    Alpha=NamedRerouteUsage()
   )
   Alpha=Saturate(None=Divide(A=@n35, B=@n36))
  )
  n38 = NamedRerouteUsage()
  Add(
   A=MaterialFunctionCall(Number=NamedRerouteUsage(), UVs=ScreenPosition().ViewportUV) {materialFunction=&DebugScalarValues}.Result
   B=ShadingPathSwitch(Default=LinearInterpolate(A=@n22.EyeAdaptationInverse, B=@n23, Alpha=@n24), Mobile=LinearInterpolate(A=@n37, B=@n28, Alpha=@n38))
  )
  n39 = ScreenPosition()
  n40 = TextureCoordinate()
  FeatureLevelSwitch(Default=@n39.ViewportUV, ES2=@n40, ES3_1=@n40, SM4_REMOVED=@n39.ViewportUV)
  Multiply(A=@n15, B=NamedRerouteUsage()) {desc=* Caustics}
  n41 = NamedRerouteUsage()
  n42 = ScalarParameter() {defaultValue=10, parameterName=Near Clip Plane}
  n43 = MaterialFunctionCall() {materialFunction=&CameraDirectionVector}
  n44 = CameraVectorWS()
  n45 = Max(A=Abs(None=DotProduct(A=@n43.Result, B=@n44))) {constB=0}
  n46 = Subtract(A=CameraPositionWS(), B=Multiply(A=CameraVectorWS(), B=Divide(A=@n42, B=@n45)))
  n47 = MaterialFunctionCall!(Texture, Position=@n46) {materialFunction=/FortWater/Materials/Functions/WaterBodyData}
  n48 = NamedRerouteUsage()
  NamedRerouteDeclaration(None=AppendVector(A=AppendVector(A=@n47.Water Z Height, B=@n41), B=@n48)) {name=Debug, nodeColor=<rgba 0.1719 1 0 1>}
  n49 = CameraVectorWS()
  n50 = Divide(
   A=ComponentMask(None=SceneTexture() {sceneTextureId=PPI_SceneDepth}.Color) {r=true}
   B=DotProduct(
    A=CameraVectorWS()
    B=MaterialFunctionCall() {materialFunction=&CameraDirectionVector}.Result
   )
  )
  n51 = ComponentMask(None=Add(A=CameraPositionWS(), B=Multiply(A=@n49, B=@n50))) {g=true, r=true}
  n52 = Add(
   A=Multiply(A=@n51, B=ScalarParameter() {defaultValue=0.0009, group=Mobile Water, parameterName=Mobile Caustics1 Scale}) {constB=0.0008}
   B=Multiply(A=Time(), B=ScalarParameter() {defaultValue=0.04, group=Mobile Water, parameterName=Mobile Caustics1 Speed}) {constB=0.03}
  )
  n53 = ComponentMask(None=@n51) {g=true}
  n54 = ComponentMask(None=@n51) {r=true}
  n55 = ScalarParameter() {defaultValue=0.0008, group=Mobile Water, parameterName=Mobile Caustics2 Scale}
  n56 = Constant2Vector() {g=1, r=0.5}
  n57 = Subtract(
   A=Multiply(A=AppendVector(A=@n53, B=@n54), B=Multiply(A=@n55, B=@n56)) {constB=0.0007}
   B=Multiply(A=Time(), B=ScalarParameter() {defaultValue=0.05, group=Mobile Water, parameterName=Mobile Caustics2 Speed}) {constB=0.02}
  )
  n58 = Multiply(A=Multiply(
    A=TextureSample(UVs=@n52) {samplerType=SAMPLERTYPE_LinearGrayscale, texture=/FortWater/Materials/WaterSurface/Mobile/Textures/T_MobileWater_Caustics}.R
    B=TextureSample(UVs=@n57) {samplerType=SAMPLERTYPE_LinearGrayscale, texture=&T_MobileWater_Caustics}.R
   )) {constB=20}
  n59 = CameraVectorWS()
  n60 = Saturate(None=Multiply(A=ComponentMask(None=@n59) {b=true}) {constB=3})
  NamedRerouteDeclaration(None=Add(A=Multiply(A=@n58, B=@n60))) {name=Mobile Caustics, nodeColor=<rgba 1 0 0.3984 1>}
  n61 = ScalarParameter() {defaultValue=10, parameterName=Near Clip Plane}
  n62 = MaterialFunctionCall() {materialFunction=&CameraDirectionVector}
  n63 = CameraVectorWS()
  n64 = Max(A=Abs(None=DotProduct(A=@n62.Result, B=@n63))) {constB=0}
  n65 = Subtract(A=CameraPositionWS(), B=Multiply(A=CameraVectorWS(), B=Divide(A=@n61, B=@n64)))
  n66 = ComponentMask(None=@n65) {b=true}
  n67 = ScalarParameter() {defaultValue=-0.25, parameterName=Water Level Z Bias}
  n68 = NamedRerouteUsage()
  n69 = MaterialFunctionCall(Position=@n65) {materialFunction=&WaterBodyData}
  n70 = MaterialFunctionCall!(Position=@n65, Water Depth=@n69.Water Depth) {materialFunction=/FortWater/Materials/WaterSurface/Mobile/Functions/MF_MobileWaterWaves}
  n71 = NamedRerouteUsage()
  n72 = Subtract(A=Multiply(A=ComponentMask(None=@n71) {a=true}) {constB=8}) {constB=7}
  n73 = Add(A=Multiply(A=ComponentMask(None=@n70.Result) {b=true}, B=Saturate(None=@n72)), B=@n69.Water Z Height)
  n74 = Divide(
   A=Subtract(A=Subtract(A=@n66, B=@n67), B=StaticSwitchParameter(True=@n68, False=@n73) {parameterName=useFixedZ})
   B=ScalarParameter() {defaultValue=0.2, parameterName=Water Edge Width}
  )
  n75 = ScalarParameter() {defaultValue=2.6346, parameterName=Wet Mask Z Bias}
  NamedRerouteDeclaration(None=Saturate(None=Subtract(A=@n74, B=@n75))) {name=Mobile WaterMask, nodeColor=<rgba 0 1 0.1562 1>}
  n76 = CameraPositionWS()
  n77 = Ceil(None=Subtract(A=ScalarParameter() {defaultValue=-3500, parameterName=OCean}, B=ComponentMask(None=@n76) {b=true}))
  n78 = WorldPosition()
  n79 = CameraPositionWS()
  n80 = CollectionParameter() {collection=/Water/Materials/MPC/MPC_Water, parameterName=SwampColor}
  n81 = ComponentMask(None=@n80) {a=true}
  n82 = ScalarParameter() {parameterName=SwampRadiusOffset}
  n83 = Ceil(None=@n81)
  n84 = SphereMask(
   A=ComponentMask(None=LinearInterpolate(A=@n78.XYZ, B=@n79) {constAlpha=0}) {g=true, r=true}
   B=ComponentMask(None=@n80) {g=true, r=true}
   Radius=Multiply(A=Add(A=@n81, B=@n82), B=Saturate(None=@n83) {desc="Cancel swamp if radius <= 0"})
   Hardness=CollectionParameter() {collection=&MPC_Water, parameterName=SwampMaskHardness}
  ) {desc=Swamp Scattering Mask}
  n85 = ScalarParameter() {defaultValue=2, parameterName=BiomePower}
  n86 = StaticSwitchParameter(
   True=Multiply(
    A=Saturate(None=@n77)
    B=MaterialFunctionCall(In=@n84, Power=@n85) {materialFunction=&SCurve}.Result
   )
   False=Constant()
  ) {defaultValue=true, parameterName=Use Swamp Biome}
  n87 = MaterialFunctionCall(Position=WorldPosition().XYZ) {materialFunction=&WaterBodyData}
  n88 = ComponentMask(None=@n87.VelocityAndLength) {b=true}
  n89 = OneMinus(None=Saturate(None=Divide(A=@n88) {constB=200}))
  NamedRerouteDeclaration(None=StaticSwitchParameter(
    True=MaterialFunctionCall!(In, isRiver_Transition, isRiver, Transition_Mask, Depth, FlowMap Input) {materialFunction=/WildEstate/Water/MaterialFunctions/MF_WildEstate_NucularWater}.MaskforPost
    False=Multiply(A=@n86, B=@n89)
   ) {parameterName=UseNucularWater}) {name=Biome Mask, nodeColor=<rgba 1 0.1406 0 1>}
  NamedRerouteDeclaration(None=@n47.Water Depth) {name=Player Loc WaterDepth, nodeColor=<rgba 0.1953 0 1 1>}
  n90 = Subtract(A=ComponentMask(None=@n47.VelocityAndLength) {b=true}, B=ScalarParameter() {defaultValue=8, parameterName=Velocity for Full Waves})
  n91 = ScalarParameter() {defaultValue=64, parameterName=Velocity for No Waves}
  n92 = OneMinus(None=Saturate(None=Divide(A=@n90, B=@n91)))
  n93 = NamedRerouteUsage()
  n94 = ScalarParameter() {parameterName=WaterBodyIndex}
  MaterialExpressionScalarParameter_13 = ScalarParameter() {parameterName=WaterBodyIndex}
  n95 = MaterialFunctionCall!(Z Offset Only, Position=@n46, WaterBodyIndex=@MaterialExpressionScalarParameter_13, Normals use VS Interpolation, All Normals VS) {materialFunction=/FortWater/Materials/Functions/Fort_ComputeGerstnerWaves}
  n96 = Multiply(
   A=Multiply(
    A=NamedRerouteDeclaration(None=@n92) {name=VelocityMask, nodeColor=<rgba 1 0 0.2812 1>}
    B=MaterialFunctionCall!(WaterDepth=@n93, Water Body Index=@n94) {materialFunction=/Water/Materials/Functions/ComputeWavesAttenuationFactor}.WaveAttenuationFactor
   )
   B=Multiply(
    A=ComponentMask(None=@n95.WPO) {b=true}
    B=MaterialFunctionCall!(Material) {materialFunction=/WildEstate/Water/MaterialFunctions/MF_WildEstate_Reflection}.WPO mask
   )
  ) {constA=1}
  n97 = StaticSwitchParameter(
   True=@n41
   False=StaticSwitchParameter(True=LinearInterpolate(A=@n47.Water Z Height, B=@n41, Alpha=@n48), False=@n47.Water Z Height) {parameterName=UseVelocityAsFixedZMask}
  ) {parameterName=useFixedZ}
  n98 = ComponentMask(None=@n46) {b=true}
  n99 = ScalarParameter() {defaultValue=-0.25, parameterName=Water Level Z Bias}
  n100 = Divide(A=Subtract(A=Subtract(A=@n98, B=@n99), B=Add(A=@n96, B=@n97)), B=ScalarParameter() {defaultValue=0.2, parameterName=Water Edge Width})
  n101 = ScalarParameter() {defaultValue=2.6346, parameterName=Wet Mask Z Bias}
  NamedRerouteDeclaration(None=Saturate(None=Subtract(A=@n100, B=@n101))) {name=Water Mask, nodeColor=<rgba 0 0.2178 1 1>}
  n102 = Constant() {r=1}
  n103 = NamedRerouteUsage()
  n104 = OneMinus(None=Saturate(None=Ceil(None=@n87.Water Depth)))
  n105 = ScalarParameter() {defaultValue=-5000, parameterName=Terrain Shift}
  n106 = LinearInterpolate(
   A=Add(A=Multiply(A=@n104, B=@n105) {constB=-1000}, B=@n87.Water Z Height)
   B=NamedRerouteUsage()
   Alpha=ScalarParameter() {parameterName=Use Forced Flat Water Surface}
  )
  n107 = WorldPosition()
  n108 = Subtract(
   A=NamedRerouteDeclaration(None=StaticSwitchParameter(True=@n103, False=@n106) {parameterName=useFixedZ}) {name=Water Z Height, nodeColor=<rgba 0 1 0.2266 1>}
   B=ComponentMask(None=@n107.XYZ) {b=true}
  )
  NamedRerouteDeclaration(None=MaterialFunctionCall!(
    WorldPosition=WorldPosition().XYZ
    Scale=ScalarParameter() {defaultValue=1536, parameterName=Caustics Scale}
    Water Depth=Multiply(A=@n102, B=@n108)
    Time
    Water Velocity=MaterialFunctionCall(Position=@n107.XYZ) {materialFunction=&WaterBodyData}.VelocityAndLength
   ) {materialFunction=/FortWater/Materials/Functions/WaterCaustics}.Result) {name=Caustic Mask, nodeColor=<rgba 0 1 0.1328 1>}
  NamedRerouteDeclaration(None=Saturate(None=@n100)) {name=Edge Mask, nodeColor=<rgba 1 0.6797 0 1>}
  n109 = Custom!(WaterBodyIndex=@MaterialExpressionScalarParameter_13) {
   code="FWaterBodyData WaterBodyData = GetWaterBodyData((int)WaterBodyIndex);\r\n\r\nFixedVelocity = WaterBodyData.FixedVelocity;\r\nFixedZHeight = WaterBodyData.FixedZHeight;\r\nFixedWaterDepth = WaterBodyData.FixedWaterDepth;\r\n\r\nreturn 0.0f;"
   description=GetWaterBodyFallbacks
   additionalOutputs [
    {outputName=FixedVelocity, outputType=CMOT_Float3}
    {outputName=FixedZHeight, outputType=CMOT_Float1}
    {outputName=FixedWaterDepth, outputType=CMOT_Float1}
   ]
   includeFilePaths [/Plugin/Water/Private/WaterDataFunctions.ush]
   inputs [{
     inputName=WaterBodyIndex
     input {expression=&self:MaterialExpressionScalarParameter_13}
    }]
  }
  NamedRerouteDeclaration(None=@n109.FixedZHeight) {name=FixedZHeight, nodeColor=<rgba 0 1 0.25 1>}
  NamedRerouteDeclaration(None=@n109.FixedWaterDepth) {name=FixedWaterDepth, nodeColor=<rgba 0.9453 0 1 1>}
  Saturate(None=@n93)
  Sign(None=ComponentMask(None=WorldPosition().XYZ) {b=true})
 ]
}
