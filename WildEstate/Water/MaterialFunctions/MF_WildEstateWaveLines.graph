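# Material function that blends stylized wave lines into water BaseColor and Opacity, with static switches for pool, river and transition variants.
# Shore lines pan a wave texture along gradients, broken up by world-space noise.
# Far lines are rippling, camera-facing dashes in noise cells that scale with distance, with a flip-flop fade between steps.
# A sphere location mask reads MPC_Water's YachtLocation, and mobile scales down line intensity.
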MaterialFunction /WildEstate/Water/MaterialFunctions/MF_WildEstateWaveLines
graph {nodes [
  n1 = FunctionInput() {inputType=FunctionInput_MaterialAttributes}
  n2 = GetMaterialAttributes(None=@n1) {attributeGetTypes [#1, #2]}
  n3 = NamedRerouteUsage()
  n4 = StaticBool()
  n5 = StaticSwitch(
   True=@n3
   False=Add(A=@n3, B=NamedRerouteUsage())
   Value=NamedRerouteDeclaration(None=FunctionInput(Preview=@n4) {bUsePreviewValueAsDefault=true, inputName=IsPoolBody, inputType=FunctionInput_StaticBool, sortPriority=50}) {name=IsPoolBody, nodeColor=<rgba 1 0.75 0 1>}
  )
  n6 = Multiply(
   A=@n5
   B=FunctionInput(Preview=Constant() {r=1}) {bUsePreviewValueAsDefault=true, inputName=TransitionMask, inputType=FunctionInput_Scalar, sortPriority=20}
  )
  n7 = StaticSwitchParameter(True=@n6, False=StaticSwitchParameter(True=@n6, False=@n5) {parameterName=Ocean River Transition}) {parameterName=River Transition}
  n8 = ShadingPathSwitch(Default=@n7, Mobile=Multiply(A=ScalarParameter() {defaultValue=0.25, parameterName=WildEstate_WaveLines_Mobile_Intensity}, B=@n7)) {desc=Migrated from FeatureLevelSwitch}
  n9 = SetMaterialAttributes!(
   MaterialAttributes=@n2.MaterialAttributes
   BaseColor=LinearInterpolate(A=@n2.BaseColor, B=VectorParameter() {parameterName=WildEstate_Color_WaveLines, defaultValue=<rgba 0.6087 0.7291 0.9167 1>}.RGB, Alpha=@n8)
   Opacity=LinearInterpolate(A=@n2.Opacity, B=ScalarParameter() {defaultValue=1, parameterName=Opacity_WildEstate_WaveLines}, Alpha=@n8)
  ) {attributeSetTypes [#1, #2]}
  n10 = StaticSwitch(True=@n1, False=@n9, Value=NamedRerouteUsage())
  FunctionOutput(None=StaticSwitchParameter(True=@n9, False=StaticSwitchParameter(True=@n9, False=@n10) {parameterName=Ocean River Transition}) {parameterName=River Transition})
  n11 = Subtract(
   A=Subtract(A=SceneDepthWithoutWater!(UVs).R, B=PixelDepth().R)
   B=ScalarParameter() {defaultValue=16, parameterName=IntersectionLine_DepthOffset}
  )
  LinearInterpolate(Alpha=Saturate(None=Round(None=@n11)))
  n12 = Add(
   A=Divide(A=NamedRerouteUsage(), B=ScalarParameter() {defaultValue=1.0067, group=ShoreWaveLines, parameterName=WildEstate_Shore_Gradient_3})
   B=NamedRerouteUsage()
  )
  n13 = Multiply(A=Time(), B=ScalarParameter() {defaultValue=0.05, group=ShoreWaveLines, parameterName=WildEstate_WavePanSpeed_3})
  n14 = Add(
   A=Add(A=@n12, B=Frac(None=@n13))
   B=Multiply(A=NamedRerouteUsage(), B=ScalarParameter() {defaultValue=5, group=ShoreWaveLines, parameterName=WildEstate_WavePanSpeed_3_Offset})
  )
  n15 = LinearInterpolate(
   A=ScalarParameter() {defaultValue=-27.3694, group=ShoreWaveLines, parameterName=WildEstate_LargeMask_C_low}
   B=ScalarParameter() {defaultValue=13.146, group=ShoreWaveLines, parameterName=WildEstate_LargeMask_C_high}
   Alpha=NamedRerouteUsage()
  )
  n16 = ScalarParameter() {defaultValue=0.7262, group=ShoreWaveLines, parameterName=WildEstate_Mask_CenterRipple_Bias}
  n17 = Saturate(None=Divide(
    A=Subtract(A=@n12, B=@n16)
    B=ScalarParameter() {defaultValue=0.1946, group=ShoreWaveLines, parameterName=WildEstate_Mask_CenterRipple_Blend}
   ))
  n18 = Multiply(A=Multiply(A=TextureSample(UVs=@n14) {texture=/WildEstate/Water/Textures/T_WildEstate_WaterWave_M}.R, B=Saturate(None=@n15)), B=@n17)
  n19 = NamedRerouteUsage()
  n20 = ScalarParameter() {defaultValue=350, group=ShoreWaveLines, parameterName=WildEstate_Shore_Gradient_max_Bias}
  n21 = Saturate(None=Divide(A=Subtract(A=@n19, B=@n20), B=ScalarParameter() {defaultValue=10, group=ShoreWaveLines, parameterName=WildEstate_Shore_Gradient_max}))
  n22 = Subtract(
   A=NamedRerouteUsage()
   B=ScalarParameter() {defaultValue=69.0604, group=ShoreWaveLines, parameterName=WildEstate_Shore_Gradient_ShallowFade_Offset}
  )
  n23 = LinearInterpolate(
   A=ScalarParameter() {defaultValue=50, group=ShoreWaveLines, parameterName=WildEstate_Shore_Gradient_ShallowFade}
   B=ScalarParameter() {defaultValue=200, group=Location_Waves, parameterName=WildEstate_Shore_Gradient_ShallowFade_4}
   Alpha=Ceil(None=NamedRerouteUsage())
  ) {constB=-0.05}
  n24 = Multiply(A=OneMinus(None=@n21), B=Saturate(None=Divide(A=@n22, B=@n23)))
  n25 = NamedRerouteUsage()
  n26 = ScalarParameter() {defaultValue=500, group=ShoreWaveLines, parameterName=WildEstate_Shore_Gradient}
  n27 = NamedRerouteUsage()
  n28 = Multiply(
   A=Time()
   B=LinearInterpolate(
    A=ScalarParameter() {defaultValue=0.05, group=ShoreWaveLines, parameterName=WildEstate_WavePanSpeed}
    B=ScalarParameter() {defaultValue=-0.05, group=Location_Waves, parameterName=WildEstate_WaveSpeed_4}
    Alpha=Ceil(None=@n27)
   ) {constB=-0.05}
  )
  n29 = Add(A=Add(A=Divide(A=@n25, B=@n26), B=Frac(None=@n28)), B=NamedRerouteUsage())
  n30 = LinearInterpolate(
   A=ScalarParameter() {defaultValue=-7.7156, group=ShoreWaveLines, parameterName=WildEstate_LargeMask_low}
   B=ScalarParameter() {defaultValue=4.2249, group=ShoreWaveLines, parameterName=WildEstate_LargeMask_high}
   Alpha=NamedRerouteUsage()
  )
  n31 = LinearInterpolate(
   A=ScalarParameter() {defaultValue=4.5378, group=ShoreWaveLines, parameterName=WildEstate_LargeMask_B_low}
   B=ScalarParameter() {defaultValue=-5, group=ShoreWaveLines, parameterName=WildEstate_LargeMask_B_high}
   Alpha=NamedRerouteUsage()
  )
  n32 = Add(A=@n29, B=ScalarParameter() {defaultValue=0.48, group=ShoreWaveLines, parameterName=WildEstate_WavePanOffset})
  n33 = Add(
   A=Multiply(A=TextureSample(UVs=@n29) {texture=&T_WildEstate_WaterWave_M}.RGB, B=Saturate(None=@n30))
   B=Multiply(A=Saturate(None=@n31), B=TextureSample(UVs=@n32) {texture=&T_WildEstate_WaterWave_M}.RGB)
  )
  n34 = ComponentMask(None=@n33) {r=true}
  NamedRerouteDeclaration(None=Add(A=Multiply(A=@n18, B=@n21), B=Multiply(A=@n24, B=@n34))) {name=ShoreWaveLines, nodeColor=<rgba 1 0 0.2578 1>}
  n35 = Constant() {r=1}
  NamedRerouteDeclaration(None=LinearInterpolate(
    A=FunctionInput(Preview=@n35) {bUsePreviewValueAsDefault=true, inputName=Water Depth, inputType=FunctionInput_Scalar, sortPriority=10}
    Alpha=NamedRerouteUsage()
   ) {constB=0.5}) {name=Base Water Depth, nodeColor=<rgba 0 0.8594 1 1>}
  n36 = ComponentMask(None=WorldPosition().XYZ) {g=true, r=true}
  n37 = VectorParameter() {group=Location_Waves, parameterName=Mask Location Radius, defaultValue=<rgba 27672 92500 7500 1>}
  n38 = ComponentMask(None=@n37.RGB) {g=true, r=true}
  n39 = ScalarParameter() {defaultValue=0.2, group=Location_Waves, parameterName=Location Mask Hardness}
  n40 = ComponentMask(None=CollectionParameter() {collection=/Water/Materials/MPC/MPC_Water, parameterName=YachtLocation}) {r=true}
  NamedRerouteDeclaration(None=LinearInterpolate(B=SphereMask(A=@n36, B=@n38, Radius=@n37.B, Hardness=@n39), Alpha=Saturate(None=@n40)) {constAlpha=1}) {name=location mask, nodeColor=<rgba 0.8047 1 0 1>}
  NamedRerouteDeclaration(None=ComponentMask(None=@n33) {g=true}) {name=ShoreWaveMask, nodeColor=<rgba 0.8281 0 1 1>}
  n41 = Divide(A=WorldPosition().XY, B=ScalarParameter() {defaultValue=7205.1553, group=ShoreWaveLines, parameterName=WildEstate_Noise_Size})
  NamedRerouteDeclaration(None=Multiply(
    A=TextureSample(UVs=@n41) {samplerType=SAMPLERTYPE_Grayscale, texture=/WildEstate/Water/Textures/WildEstate_Water_M}.R
    B=ScalarParameter() {defaultValue=0.15, group=ShoreWaveLines, parameterName=WildEstate_WaveDistortion}
   )) {name=Noise Distortion Small, nodeColor=<rgba 1 0.8203 0 1>}
  n42 = ComponentMask(None=NamedRerouteUsage()) {r=true}
  n43 = ScalarParameter() {defaultValue=0.8157, group=WaveLinesFar, parameterName=Threshold}
  n44 = ComponentMask(None=Add(A=NamedRerouteUsage()) {constB=0.5}) {g=true, r=true}
  n45 = ScalarParameter() {defaultValue=0.646, group=WaveLinesFar, parameterName=LineScaleMin}
  n46 = ScalarParameter() {defaultValue=1.5223, group=WaveLinesFar, parameterName=LineScaleMax}
  n47 = ComponentMask(None=NamedRerouteUsage()) {g=true}
  n48 = Constant()
  n49 = ScalarParameter() {defaultValue=0.03, group=WaveLinesFar, parameterName=LineSineWaveRippleAmplitude}
  n50 = NamedRerouteUsage()
  n51 = ScalarParameter() {defaultValue=0.664, group=WaveLinesFar, parameterName=LineSineWaveRipplefrequency}
  n52 = ScalarParameter() {defaultValue=0.528, group=WaveLinesFar, parameterName=LineSineWaveRipplefrequency2}
  n53 = ComponentMask(None=NamedRerouteUsage()) {g=true}
  n54 = NamedRerouteUsage()
  n55 = Add(A=Time(), B=Multiply(A=ComponentMask(None=@n54) {g=true}) {constB=33.3})
  n56 = ScalarParameter() {defaultValue=0.24, group=WaveLinesFar, parameterName=LineSineWaveRippleSpeed}
  n57 = Add(A=Divide(A=ComponentMask(None=@n50) {r=true}, B=LinearInterpolate(A=@n51, B=@n52, Alpha=@n53)), B=Frac(None=Multiply(A=@n55, B=@n56)))
  n58 = Add(
   A=MaterialFunctionCall(UVs=Saturate(None=@n44), Texture Scale=LinearInterpolate(A=@n45, B=@n46, Alpha=@n47)) {materialFunction=&ScaleUVsByCenter}.UVs
   B=Multiply(A=AppendVector(A=@n48, B=@n49), B=Sine(None=@n57))
  )
  n59 = Multiply(
   A=ComponentMask(None=TextureSample(UVs=@n58) {texture=/WildEstate/Water/Textures/T_WildEstate_WaterSmallLine}.RGB) {r=true}
   B=NamedRerouteUsage()
  )
  n60 = ScalarParameter() {defaultValue=0.5, group=WaveLinesFar, parameterName=SmallLinesOpaicty}
  n61 = Multiply(A=Ceil(None=Subtract(A=@n42, B=@n43)), B=Saturate(None=Multiply(A=@n59, B=@n60)))
  n62 = NamedRerouteUsage()
  n63 = ScalarParameter() {defaultValue=500, group=WaveLinesFar, parameterName=ShreWaveMasking}
  n64 = Saturate(None=OneMinus(None=Multiply(A=@n62, B=@n63)))
  NamedRerouteDeclaration(None=StaticSwitch(True=@n61, False=Multiply(A=@n61, B=@n64), Value=NamedRerouteUsage())) {name=SmallWaveLinesFar, nodeColor=<rgba 0 1 0.5078 1>}
  n65 = WorldPosition()
  n66 = ScalarParameter() {defaultValue=27890.623, group=ShoreWaveLines, parameterName=WildEstate_Noise_Size2}
  NamedRerouteDeclaration(None=TextureSample(UVs=Divide(A=@n65.XY, B=@n66)) {samplerType=SAMPLERTYPE_Grayscale, texture=&WildEstate_Water_M}.R) {name=Noise Large, nodeColor=<rgba 1 0.8203 0 1>}
  n67 = WorldPosition()
  n68 = ScalarParameter() {defaultValue=12532.082, group=ShoreWaveLines, parameterName=WildEstate_Noise_Size3}
  NamedRerouteDeclaration(None=TextureSample(UVs=Divide(A=@n67.XY, B=@n68)) {samplerType=SAMPLERTYPE_Grayscale, texture=&WildEstate_Water_M}.R) {name=Noise Medium, nodeColor=<rgba 1 0.8203 0 1>}
  n69 = Divide(
   A=Length(None=WorldPosition() {worldPositionShaderOffset=WPT_CameraRelativeNoOffsets}.XYZ) {is=class MaterialExpressionLength}
   B=ScalarParameter() {defaultValue=5000, group=WaveLinesFar, parameterName=FadeBias}
  )
  n70 = Divide(
   A=AppendVector(A=WorldPosition().XY, B=Constant())
   B=Multiply(A=ScalarParameter() {defaultValue=150.2361, group=WaveLinesFar, parameterName=Grass_Card_CellSize}, B=Ceil(None=@n69))
  )
  NamedRerouteDeclaration(None=VectorNoise(World Position=@n70)) {name=CellNoise, nodeColor=<rgba 0.3828 1 0 1>}
  n71 = Frac(None=@n69)
  n72 = Multiply(A=@n71, B=OneMinus(None=@n71))
  n73 = ScalarParameter() {defaultValue=4, group=WaveLinesFar, parameterName=FlipFlopTransition}
  NamedRerouteDeclaration(None=Saturate(None=Multiply(A=@n72, B=@n73))) {name=FlipFlopTransition, nodeColor=<rgba 1 0 0.7031 1>}
  n74 = Subtract(A=Frac(None=@n70)) {constB=0.5}
  n75 = Constant3Vector() {constant=<rgba 1 0 0 0>}
  n76 = Normalize(VectorInput=Multiply(
    A=Transform(None=@n75) {transformSourceType=TRANSFORMSOURCE_Camera, transformType=TRANSFORM_Tangent}
    B=Constant3Vector() {constant=<rgba 1 1 0 0>}
   ))
  n77 = Transform(None=Constant3Vector() {constant=<rgba 0 0 1 0>}) {transformSourceType=TRANSFORMSOURCE_Camera, transformType=TRANSFORM_Tangent}
  n78 = Constant3Vector() {constant=<rgba 1 1 0 0>}
  n79 = DotProduct(A=@n74, B=Normalize(VectorInput=Multiply(A=@n77, B=@n78)))
  NamedRerouteDeclaration(None=AppendVector(A=DotProduct(A=@n74, B=@n76), B=Multiply(A=@n79) {constB=-1})) {name=CamerFacingUVCells, nodeColor=<rgba 0 1 0.9766 1>}
  NamedRerouteDeclaration(None=FunctionInput(Preview=StaticBool()) {bUsePreviewValueAsDefault=true, inputName=IsRiver, inputType=FunctionInput_StaticBool, sortPriority=20}) {name=isRiver, nodeColor=<rgba 0.0781 0 1 1>}
  OneMinus(None=@n17)
  ScalarParameter() {group=Location_Waves, parameterName=LocationMask_Enabled}
 ]}
