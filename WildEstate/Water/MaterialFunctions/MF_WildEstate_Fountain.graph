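# Fountain water material function that replaces the incoming BaseColor with a two-color animated ripple pattern.
# Local-space polar UVs scroll distorted noise with radial edge falloff, using ripple settings chosen for three bowl tiers by local Z.
# Mirrored, noise-perturbed sine waves around an offset add splash ripples to the bottom tier.
# A narrow threshold turns the height into the color mask; the derived normal and reflection chain never reaches the output.
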MaterialFunction /WildEstate/Water/MaterialFunctions/MF_WildEstate_Fountain
graph {nodes [
  CameraVectorWS()
  n1 = FunctionInput() {inputType=FunctionInput_MaterialAttributes}
  n2 = LinearInterpolate(
   A=VectorParameter() {group=Fountain, parameterName=Ripple Low Color, defaultValue=<rgba 0 0.166 0.3448 1>}.RGB
   B=VectorParameter() {group=Fountain, parameterName=Ripple High Color, defaultValue=<rgba 0.0764 0.2538 0.4448 1>}.RGB
   Alpha=NamedRerouteUsage()
  )
  FunctionOutput(None=SetMaterialAttributes!(MaterialAttributes=GetMaterialAttributes(None=@n1).MaterialAttributes, BaseColor=@n2) {attributeSetTypes [#1]})
  n3 = LocalPosition()
  n4 = MaterialFunctionCall(Float2=@n3.XY) {materialFunction=&BreakOutFloat2Components}
  n5 = Arctangent2Fast(Y=@n4.R, X=@n4.R)
  n6 = AppendVector(A=Length(None=@n3.XY) {is=class MaterialExpressionLength}, B=Add(A=Divide(A=@n5) {constB=6.2832}) {constB=0.5})
  n7 = VectorParameter() {group=Fountain, parameterName=Main Noise Scale + Speed, defaultValue=<rgba 0.04 1 -0.4 0>}
  n8 = MaterialFunctionCall(Float2=Multiply(A=@n6, B=ComponentMask(None=@n7.RGBA) {g=true, r=true})) {materialFunction=&BreakOutFloat2Components}
  n9 = AppendVector(A=Power(Base=@n8.R) {constExponent=0.75}, B=@n8.G)
  n10 = Multiply(A=ComponentMask(None=@n7.RGBA) {a=true, b=true}, B=Time())
  n11 = LocalPosition()
  n12 = Add(A=@n11.Z, B=ScalarParameter() {defaultValue=-100, group=Fountain, parameterName=Ripple Bowl Height 1})
  n13 = OneMinus(None=Ceil(None=Saturate(None=@n12)))
  n14 = VectorParameter() {group=Fountain, parameterName=Bottom Ripple Scale + Speed + Edge Falloff, defaultValue=<rgba 0.04 7 -0.6 170>}
  n15 = Add(A=@n11.Z, B=ScalarParameter() {defaultValue=-300, group=Fountain, parameterName=Ripple Bowl Height 2})
  n16 = OneMinus(None=Ceil(None=Saturate(None=@n15)))
  n17 = VectorParameter() {group=Fountain, parameterName=Middle Ripple Scale + Speed + Edge Falloff, defaultValue=<rgba 0.04 5 -0.1 60>}
  n18 = VectorParameter() {group=Fountain, parameterName=Top Ripple Scale + Speed + Edge Falloff, defaultValue=<rgba 0.04 3 -0.08 30>}
  n19 = MaterialFunctionCall(Float4=If(A=@n13, "A > B"=@n14.RGBA, "A < B"=If(A=@n16, "A > B"=@n17.RGBA, "A < B"=@n18.RGBA) {constB=0.1}) {constB=0.1}) {materialFunction=&BreakOutFloat4Components}
  n20 = MaterialFunctionCall(Float2=Multiply(A=@n6, B=AppendVector(A=@n19.R, B=@n19.R))) {materialFunction=&BreakOutFloat2Components}
  n21 = Power(Base=@n20.R) {constExponent=1}
  n22 = AppendVector(A=@n19.B, B=Constant())
  n23 = Time()
  n24 = TextureSample!(UVs=Add(A=AppendVector(A=@n21, B=@n20.G), B=Multiply(A=@n22, B=@n23)), Tex, Level, Apply View MipBias) {mipValueMode=TMVM_MipLevel, samplerType=SAMPLERTYPE_Grayscale, texture=/Game/Global/Textures/Default/Noise/LowResBlurredNoiseNormalized}
  n25 = ConstantBiasScale(None=@n24.R) {bias=-0.5, scale=2}
  n26 = AppendVector(A=ScalarParameter() {defaultValue=0.05, group=Fountain, parameterName=Secondary Noise Strength}, B=Constant())
  n27 = TextureSample(UVs=Add(A=Add(A=@n9, B=@n10), B=Multiply(A=@n25, B=@n26))) {mipValueMode=TMVM_MipLevel, samplerType=SAMPLERTYPE_Grayscale, texture=&LowResBlurredNoiseNormalized}
  n28 = Divide(A=Length(None=LocalPosition().XY), B=@n19.A)
  n29 = Add(
   A=@n27.R
   B=Subtract(A=Power(Base=@n28) {constExponent=0.9}, B=ScalarParameter() {defaultValue=0.85, group=Fountain, parameterName=Ripple Edge Distance})
  )
  n30 = ComponentMask(None=VectorParameter() {group=Fountain, parameterName=External Ripple Offset, defaultValue=<rgba -200 0 0 1>}.RGB) {g=true, r=true}
  n31 = Max(A=Add(A=@n3.XY, B=@n30) {constB=0}, B=Add(A=Multiply(A=@n3.XY) {constB=-1}, B=@n30))
  n32 = MaterialFunctionCall(Float2=@n31) {materialFunction=&BreakOutFloat2Components}
  n33 = Arctangent2Fast(Y=@n32.R, X=@n32.R)
  n34 = AppendVector(A=Length(None=@n31), B=Add(A=Divide(A=@n33) {constB=6.2832}) {constB=0.5})
  n35 = VectorParameter() {group=Fountain, parameterName=External Wave Scale + Power + Period + Speed, defaultValue=<rgba 0.003 0.5 5 -0.75>}
  n36 = Power(Base=Multiply(A=ComponentMask(None=@n34) {r=true}, B=@n35.R) {constB=0.003}, Exp=@n35.G) {constExponent=0.5}
  n37 = VectorParameter() {group=Fountain, parameterName=External Noise Scale + Speed, defaultValue=<rgba 0.02 4 -0.4 0>}
  n38 = MaterialFunctionCall(Float2=Multiply(A=@n34, B=ComponentMask(None=@n37.RGBA) {g=true, r=true})) {materialFunction=&BreakOutFloat2Components}
  n39 = Power(Base=@n38.R) {constExponent=0.75}
  n40 = ComponentMask(None=@n37.RGBA) {a=true, b=true}
  n41 = DotProduct(A=@n3.XY, B=Normalize(VectorInput=@n30))
  n42 = ScalarParameter() {defaultValue=100, parameterName=External Wave Mirror Padding}
  n43 = Saturate(None=Abs(None=Divide(A=@n41, B=@n42) {constB=100}))
  n44 = ScalarParameter() {defaultValue=0.7, parameterName=External Wave Mirror Offset}
  n45 = Add(A=Time(), B=Multiply(A=Saturate(None=@n41), B=Multiply(A=@n43, B=@n44)))
  n46 = TextureSample(UVs=Add(A=AppendVector(A=@n39, B=@n38.G), B=Multiply(A=@n40, B=@n45))) {mipValueMode=TMVM_MipLevel, samplerType=SAMPLERTYPE_Grayscale, texture=&LowResBlurredNoiseNormalized}
  n47 = ConstantBiasScale(None=@n46.R) {bias=-0.5, scale=2}
  n48 = ScalarParameter() {defaultValue=0.15, group=Fountain, parameterName=External Noise Strength}
  n49 = Add(A=Add(A=Multiply(A=@n36, B=@n35.B) {constB=5}, B=Multiply(A=@n47, B=@n48)), B=Multiply(A=@n35.A, B=@n45) {constA=-0.75})
  n50 = Power(Base=ConstantBiasScale(None=Sine(None=@n49)), Exp=ScalarParameter() {defaultValue=2, group=Fountain, parameterName=Final External Power})
  n51 = VectorParameter() {group=Fountain, parameterName="External Ripple Falloff Size + Cutoff (Innter, Outer)", defaultValue=<rgba 100 0.5 300 0.2>}
  n52 = ComponentMask(None=@n34) {r=true}
  n53 = InverseLinearInterpolate(A=@n51.A, Value=Divide(A=@n52, B=@n51.B))
  n54 = Power(
   Base=OneMinus(None=Saturate(None=@n53))
   Exp=ScalarParameter() {defaultValue=1.5, group=Fountain, parameterName=External Ripple Outer Falloff Power}
  ) {constExponent=1.5}
  n55 = InverseLinearInterpolate(A=@n51.G, Value=Divide(A=@n52, B=@n51.R))
  n56 = Add(
   A=Multiply(A=Multiply(A=@n50, B=@n54), B=ScalarParameter() {defaultValue=2, group=Fountain, parameterName=Final External Multiplier})
   B=OneMinus(None=Saturate(None=@n55))
  )
  n57 = LinearInterpolate(
   A=@n29
   B=Add(A=@n29, B=Multiply(A=@n56, B=@n13)) {constB=0}
   Alpha=ScalarParameter() {defaultValue=1, group=Fountain, parameterName=UseSplashRipples}
  )
  n58 = MaterialFunctionCall(Float3=MaterialFunctionCall!(Height=@n57, Absolute World Position, World Space Vertex Normals) {materialFunction=/Engine/Functions/Engine_MaterialFunctions03/Procedurals/HeightToNormalSmooth}.WorldSpaceNormal) {materialFunction=&BreakOutFloat3Components}
  n59 = Multiply(A=@n58.B, B=ScalarParameter() {defaultValue=1, group=Fountain, parameterName=Fountain Normal Strength})
  n60 = Transform(None=Normalize(VectorInput=MaterialFunctionCall(X=@n58.R, Y=@n58.R, Z=@n59) {materialFunction=&MakeFloat3}.Result))
  Multiply(A=DotProduct(A=SkyAtmosphereLightDirection(), B=ReflectionVectorWS(CustomWorldNormal=@n60))) {constB=0.5}
  n61 = ScalarParameter() {defaultValue=0.7, group=Fountain, parameterName=Ripple Cutoff}
  n62 = Add(A=@n61) {constB=0.02}
  NamedRerouteDeclaration(None=Saturate(None=InverseLinearInterpolate(A=@n61, B=@n62, Value=@n57) {constA=0.5})) {name=Ripples, nodeColor=<rgba 0.2656 0 1 1>}
  SetMaterialAttributes!(MaterialAttributes, BaseColor=@n2, Metallic=Constant(), Specular=Constant(), Roughness=Constant() {r=1}) {attributeSetTypes [
    #1
    #2
    #3
    #4
   ]}
 ]}
