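# Water material function for WildEstate that layers fluid simulation textures onto surface normal, world position offset and foam.
# It maps world XY around MPC_Water's SimLocation to sample NormalAndHeight and Forces, quantizing the sim normal before blending it with the input.
# Flowmapped foam and optional pondscum are colored, with sim-driven erasure, waterline and debug-grid mask outputs.
# A dev scalar blends in the new shallow-water path; mobile and some quality levels bypass sim normals and displacement.
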MaterialFunction /WildEstate/Water/MaterialFunctions/WaterSimulationTexture_WildEstate
graph {nodes [
  n1 = ComponentMask(None=WorldPosition().XYZ) {g=true, r=true}
  n2 = ComponentMask(None=CollectionParameter() {collection=/Water/Materials/MPC/MPC_Water, parameterName=SimLocation}) {g=true, r=true}
  n3 = Add(A=Divide(A=Subtract(A=@n1, B=@n2), B=CollectionParameter() {collection=&MPC_Water, parameterName=FluidSimSize})) {constB=0.5}
  n4 = Saturate(None=@n3)
  n5 = TextureSampleParameter2D(UVs=@n4) {group=Fluid Simulation, parameterName=NormalAndHeight, samplerSource=SSM_Wrap_WorldGroupSettings, texture=&Black}
  n6 = DotProduct(A=@n5.RGB, B=@n5.RGB)
  n7 = ScalarParameter() {defaultValue=4, parameterName=SimNormalQuantizeSteps}
  n8 = Multiply(A=@n5.RGB, B=@n7)
  n9 = MaterialFunctionCall(Vector=Divide(A=Round(None=@n8), B=@n7), Default=Constant3Vector() {constant=<rgba 0 0 1 0>}) {materialFunction=&SafeNormalize}
  n10 = If(A=@n6, "A > B"=@n9.Result, "A == B"=Constant3Vector() {constant=<rgba 0 0 1 0>}, "A < B"=@n9.Result)
  n11 = ScalarParameter() {defaultValue=1500, group=Foam and Pondscum, parameterName=Pondscum Tiling}
  n12 = Divide(A=ComponentMask(None=WorldPosition() {worldPositionShaderOffset=WPT_ExcludeAllShaderOffsets}.XYZ) {g=true, r=true}, B=@n11)
  Add(
   A=Multiply(
    A=ScalarParameter() {defaultValue=-0.05, group=Foam and Pondscum, parameterName=Pondscum Normal Distortion}
    B=ComponentMask(None=@n10) {g=true, r=true}
   )
   B=@n12
  )
  n13 = ScalarParameter() {defaultValue=0.7, group=Flowmaps, parameterName=Foam Phase Fade Time}
  Add(A=@n13) {constB=0}
  Constant()
  n14 = AppendVector(A=ComponentMask(None=@n10) {g=true, r=true}, B=Constant())
  n15 = FunctionInput() {bUsePreviewValueAsDefault=true, inputName=Normal, sortPriority=2, previewValue=<0 0 1 1>}
  n16 = MaterialFunctionCall(Normal=@n10, Flatness=ScalarParameter() {parameterName=SimNormalFlatness}) {materialFunction=&FlattenNormal}
  n17 = ScalarParameter() {parameterName=DEV_UseNewShallowWaterSubsystem, sliderMax=1}
  n18 = QualitySwitch(
   Default=LinearInterpolate(
    A=Add(A=@n14, B=@n15)
    B=MaterialFunctionCall(BaseNormal=@n16.Result, AdditionalNormal=@n15) {materialFunction=&BlendAngleCorrectedNormals}.Result
    Alpha=@n17
   )
   Low=@n15
   High=@n15
   Medium=@n15
  )
  n19 = DotProduct(
   A=FunctionOutput(None=ShadingPathSwitch(Default=@n18, Mobile=@n15) {desc=Migrated from FeatureLevelSwitch}) {outputName=Normal}
   B=Normalize(VectorInput=VectorParameter() {group=Lighting, parameterName=SunAngle, defaultValue=<rgba 0 1 0.2 0>}.RGB)
  )
  n20 = TextureSampleParameter2D(UVs=@n4) {group=Fluid Simulation, parameterName=Forces, samplerSource=SSM_Wrap_WorldGroupSettings, texture=&Black}
  n21 = OneMinus(None=@n20.A)
  n22 = Multiply(A=Saturate(None=Ceil(None=@n21)), B=ScalarParameter() {defaultValue=0.01, parameterName=Debug Force Box})
  n23 = MaterialFunctionCall!(Alpha=@n21, Offset=ScalarParameter() {defaultValue=0.5, parameterName=Debug Force Alpha Offset}) {materialFunction=/Engine/Functions/Engine_MaterialFunctions01/Opacity/AlphaOffset}
  n24 = Multiply(A=Abs(None=Subtract(A=@n4) {constB=0.5})) {constB=2}
  n25 = MaterialFunctionCall(Float2=Saturate(None=Floor(None=@n24))) {materialFunction=&BreakOutFloat2Components}
  n26 = Max(A=@n25.R, B=@n25.R)
  n27 = Multiply(A=Abs(None=Subtract(A=@n3) {constB=0.5})) {constB=1.975}
  n28 = MaterialFunctionCall(Float2=Saturate(None=Floor(None=@n27))) {materialFunction=&BreakOutFloat2Components}
  n29 = OneMinus(None=Max(A=@n28.R, B=@n28.R))
  n30 = Multiply(
   A=Add(A=Add(A=@n22, B=@n23.Result), B=Multiply(A=@n26, B=@n29))
   B=CollectionParameter() {collection=&MPC_Water, parameterName=DebugSimGrid}
  )
  n31 = VectorParameter() {group=Foam and Pondscum, parameterName=Foam Emissive, defaultValue=<rgba 0.3 1 0.75 2>}
  n32 = FunctionInput() {bUsePreviewValueAsDefault=true, inputName=Flowmap Velocity, inputType=FunctionInput_Vector2}
  n33 = MaterialFunctionCall() {materialFunction=/Water/Materials/Functions/GetWaterTime}
  n34 = ComponentMask(None=MaterialFunctionCall!(
    Diffuse=TextureObject() {texture=/FortWater/Textures/T_Foam_DistanceField_Fade_02}
    Normal Map
    "Flow Vector Map (see tooltip)"=Divide(A=@n32, B=@n11)
    UV texture 2 offset and Scale
    "Time (see tooltip)"=MaterialFunctionCall(Time=@n33.WaterTime, Speed=@n13) {materialFunction=&TimeWithSpeedVariable}.Time
    UVs=@n12
   ) {materialFunction=/FortWater/Materials/Functions/WaterFlowMap}.Diffuse) {r=true}
  n35 = LinearInterpolate(
   B=Add(A=ScalarParameter() {defaultValue=-0.5, group=Foam and Pondscum, parameterName=Foam Texture Offset}, B=@n34)
   Alpha=ScalarParameter() {defaultValue=0.65, group=Foam and Pondscum, parameterName=foamcontrat}
  ) {constA=1}
  n36 = Multiply(
   A=Multiply(A=ConstantBiasScale(None=@n19), B=FunctionOutput(None=@n30) {outputName=Foam Mask, sortPriority=3})
   B=Multiply(A=Multiply(A=@n31.RGB, B=@n31.RGB), B=@n35)
  )
  n37 = VectorParameter() {group=Foam and Pondscum, parameterName=Pondscum Emissive, defaultValue=<rgba 0.5 0.7 0 0>}
  n38 = Multiply(A=@n37.RGB, B=@n37.RGB)
  n39 = VectorParameter() {group=Foam and Pondscum, parameterName=Pondscum Emissive2, defaultValue=<rgba 0 1 1 1>}
  n40 = Multiply(A=@n39.RGB, B=@n39.RGB)
  n41 = ScalarParameter() {defaultValue=-0.6, group=Foam and Pondscum, parameterName=PondScum Texture Offset}
  n42 = Saturate(None=Multiply(A=Add(A=@n34, B=@n41), B=ScalarParameter() {defaultValue=1, group=Foam and Pondscum, parameterName=Second Pond Color Sharpness}))
  n43 = OneMinus(None=@n5.A)
  n44 = SphereMask(
   A=LinearInterpolate(A=@n43, B=@n5.A, Alpha=ScalarParameter() {parameterName=DEV_UseNewShallowWaterSubsystem, sliderMax=1})
   B=ScalarParameter() {defaultValue=0.9, group=Foam and Pondscum, parameterName=Existing Foam Erasure Offset}
   Radius=ScalarParameter() {defaultValue=0.75, group=Foam and Pondscum, parameterName=Foam Erasure Radius}
  ) {hardnessPercent=0}
  n45 = Multiply(A=@n44, B=ScalarParameter() {defaultValue=1.5, group=Foam and Pondscum, parameterName=Foam Erasure})
  n46 = Max(A=Subtract(A=@n35, B=@n45)) {constB=0}
  n47 = Multiply(
   A=StaticSwitch(
    True=LinearInterpolate(A=@n38, B=@n40, Alpha=@n42)
    False=@n38
    Value=StaticBoolParameter() {group=Foam and Pondscum, parameterName=Use Water Pondscum Second Color}
   )
   B=@n46
  )
  FunctionOutput(None=StaticSwitch(True=Add(A=@n36, B=@n47), False=@n36, Value=StaticBoolParameter() {group=Foam and Pondscum, parameterName=Use Water Pondscum Layer})) {outputName=Foam Colorized, sortPriority=2}
  n48 = ScalarParameter() {defaultValue=16, parameterName=Max Displacement}
  n49 = Clamp(None=@n5.B, Min=Multiply(A=@n48) {constB=-1}, Max=@n48)
  n50 = VectorParameter() {group=Fluid Simulation, parameterName=FluidSimulationDisplacement, defaultValue=<rgba 0 0 0.25 0>}
  n51 = FunctionInput() {bUsePreviewValueAsDefault=true, inputName=WPO, sortPriority=3}
  n52 = Add(A=LinearInterpolate(A=Multiply(A=@n49, B=@n50.RGB), Alpha=@n17) {constB=0}, B=@n51)
  FunctionOutput(None=ShadingPathSwitch(Default=QualitySwitch(Default=@n52, Low=@n51, Medium=@n51), Mobile=@n51) {desc=Migrated from FeatureLevelSwitch}) {outputName=WPO, sortPriority=1}
  FunctionOutput(None=Saturate(None=@n45)) {outputName=Sim Foam Erasure, sortPriority=2}
  n53 = OneMinus(None=@n5.B)
  n54 = ScalarParameter() {defaultValue=-0.0339, parameterName=SimFoamNormalThreshold}
  n55 = ScalarParameter() {defaultValue=3, parameterName=SimFoamNormalSteps}
  n56 = Round(None=Multiply(A=Subtract(A=@n53, B=@n54), B=@n55))
  n57 = Saturate(None=LinearInterpolate(
    A=ScalarParameter() {defaultValue=-0.152, parameterName=SimFoamNormal_Min}
    B=ScalarParameter() {defaultValue=0.824, parameterName=SimFoamNormal_Max}
    Alpha=Divide(A=@n56, B=@n55)
   ))
  FunctionOutput(None=If(A=@n6, "A > B"=@n57, "A == B"=Constant(), "A < B"=@n57)) {outputName=SimFoamWaterline}
  Length(None=ComponentMask(None=@n10) {g=true, r=true}) {is=class MaterialExpressionLength}
  Reroute(None=@n44) {desc=hook this up to renable}
  ScalarParameter() {defaultValue=1024, group=Fluid Simulation, parameterName=FluidSize}
  n58 = Add(
   A=Multiply(A=@n43, B=ScalarParameter() {group=Technical, parameterName=ShowSimBoundary})
   B=MaterialFunctionCall(Float2=FunctionInput() {bUsePreviewValueAsDefault=true, inputName=Previous Foam, inputType=FunctionInput_Vector2, sortPriority=1}) {materialFunction=&BreakOutFloat2Components}.R
  )
  n59 = Saturate(None=Ceil(None=@n37.A))
  StaticSwitch(True=Add(A=@n58, B=Multiply(A=@n59, B=@n46)), False=@n58, Value=StaticBoolParameter() {parameterName=Use Water Pondscum Layer})
  VectorParameter() {group=Fluid Simulation, parameterName=SimLocation}
 ]}
