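# Two-sided Single Layer Water master material for Wild Estate pools, with a static switch for a fountain variant.
# Two-texture panner normals feed the water simulation function, whose normal, swappable for vertex-color flow motion, and WPO drive MF_WildEstateWaterShading.
# The shaded result goes either to fountain shading or through NucularWater, using water body height as depth, into pool-mode wave lines.
# The single-layer water output uses constant scattering and phase, with absorption via reroute.
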Material /WildEstate/Water/BaseMaterials/M_Water_Pool_WildEstate
graph {
 properties {
  allowTranslucentCustomDepthWrites=true
  bUseMaterialAttributes=true
  bUsedWithInstancedStaticMeshes=true
  bUsedWithNanite=true
  bUsedWithSplineMeshes=true
  bUsedWithWater=true
  pixelDepthOffsetMode=PDOM_Legacy
  refractionMethod=RM_PixelNormalOffset
  shadingModel=MSM_SingleLayerWater
  twoSided=true
  usedShadingModels=SingleLayerWater
 }
 nodes [
  MaterialFunctionCall!(Mask, Scattering, Absorption) {materialFunction=/Water/Materials/Functions/WaterCoefficientMask}
  n1 = MaterialFunctionCall!(Normal, WPO, Far Mask Override=ScalarParameter() {defaultValue=1, parameterName=FarMask}, Additional Mask, LowDetailBiome_Mask) {materialFunction=/FortWater/Materials/Functions/WaterTwoTexturePanner}
  n2 = MaterialFunctionCall!(Flowmap Velocity, Previous Foam, Normal=@n1.Normal, WPO=@n1.Normal) {materialFunction=/WildEstate/Water/MaterialFunctions/WaterSimulationTexture_WildEstate}
  n3 = MaterialFunctionCall!(
   Normal=StaticSwitchParameter(True=MaterialFunctionCall() {materialFunction=/FortWater/Materials/Functions/WaterPoolVertexColorFlowMotion}.Result, False=@n2.Normal) {parameterName=UseVertexColorFlowMotion}
   WPO=@n2.WPO
   TransitionMask
   IsRiver
   Depth
   IsHLOD Material
  ) {materialFunction=/WildEstate/Water/MaterialFunctions/MF_WildEstateWaterShading}
  NamedRerouteDeclaration(None=@n3.Absorption) {name=Absorption_WildEstate, nodeColor=<rgba 1 0 0.6328 1>}
  SingleLayerWaterMaterialOutput!(ScatteringCoefficients=Constant(), AbsorptionCoefficients=NamedRerouteUsage(), PhaseG=Constant() {r=0.5}, ColorScaleBehindWater)
  n4 = MaterialFunctionCall!(Texture, Position) {materialFunction=/FortWater/Materials/Functions/WaterBodyData}
  StaticSwitchParameter(
   True=MaterialFunctionCall!(In=@n3.Result) {materialFunction=/WildEstate/Water/MaterialFunctions/MF_WildEstate_Fountain}.Result
   False=MaterialFunctionCall!(
    In=MaterialFunctionCall!(In=@n3.Result, isRiver_Transition, isRiver, Transition_Mask, Depth=@n4.Water Z Height, FlowMap Input) {materialFunction=/WildEstate/Water/MaterialFunctions/MF_WildEstate_NucularWater}.Result
    Water Depth
    IsRiver
    TransitionMask
    IsPoolBody=StaticBool() {value=true}
   ) {materialFunction=/WildEstate/Water/MaterialFunctions/MF_WildEstateWaveLines}.Result
  ) {group=Fountain, parameterName=Use Fountain Shading}
 ]
}
